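# Sound cue for entering the PlunderFlop secret passage hideable, going by its name and folder.
# A short fixed delay leads into a single close-explosion wave reused from the Ashton Hippo LTM sounds, pitched down and boosted in volume.
# It uses the trap sound class and attenuation and primes on load.
# A concurrency override allows one instance per owner and rejects new plays while it sounds.
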SoundCue /Game/Sounds/Fort_Hideables/SecretPassage/Hideable_SecretPassage_PlunderFlop_Enter_Cue
root=Delay_0
nodes [
 {id=Delay_0, children [WavePlayer_0], properties {delayMax=0.25, delayMin=0.25}}
 {id=WavePlayer_0, properties {soundWaveAssetPtr=/Game/Athena/Sounds/LTMs/Ashton/Hippo/Ashton_Hippo_Explo_Close_04}}
]
properties {
 attenuationSettings=&Fort_Traps_Att
 bOverrideConcurrency=true
 bPrimeOnLoad=true
 maxDistance=6000
 pitchMultiplier=0.8
 soundClassObject=&Fort_Traps
 volumeMultiplier=1.7
 concurrencyOverrides {bLimitToOwner=true, maxCount=1, resolutionRule=PreventNew}
}
