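# Sound cue for exiting a PlunderFlop secret passage hideable, playing a muddy splat through trap sound class and attenuation.
# A mixer layers two equally weighted random picks: a boosted swamp mud blob and a large, close mud explosion burst.
# The burst passes through a modulator that pins pitch at normal.
# Concurrency override allows one instance per owner and rejects new plays, and the cue primes on load.
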SoundCue /Game/Sounds/Fort_Hideables/SecretPassage/Hideable_SecretPassage_PlunderFlop_Exit_Cue
root=Mixer_0
nodes [
 {id=Mixer_0, children [Random_1,Modulator_0], properties {inputVolume [1.5,1]}}
 {id=Modulator_0, children [Random_2], properties {pitchMax=1, pitchMin=1}}
 {
  id=Random_1
  children [WavePlayer_2, WavePlayer_8, WavePlayer_9, WavePlayer_10, WavePlayer_11, WavePlayer_12]
  properties {weights [1,1,1,1,1,1]}
 }
 {id=Random_2, children [WavePlayer_17, WavePlayer_18, WavePlayer_19, WavePlayer_20], properties {weights [1,1,1,1]}}
 {id=WavePlayer_10, properties {soundWaveAssetPtr=/Game/Sounds/Fort_Swamp_Sounds/MudBlobA004}}
 {id=WavePlayer_11, properties {soundWaveAssetPtr=/Game/Sounds/Fort_Swamp_Sounds/MudBlobA005}}
 {id=WavePlayer_12, properties {soundWaveAssetPtr=/Game/Sounds/Fort_Swamp_Sounds/MudBlobA006}}
 {id=WavePlayer_17, properties {soundWaveAssetPtr=/Game/Sounds/Fort_Weapons/Impacts/Mud/Mud_Explo_Large_Burst_Close_01}}
 {id=WavePlayer_18, properties {soundWaveAssetPtr=/Game/Sounds/Fort_Weapons/Impacts/Mud/Mud_Explo_Large_Burst_Close_02}}
 {id=WavePlayer_19, properties {soundWaveAssetPtr=/Game/Sounds/Fort_Weapons/Impacts/Mud/Mud_Explo_Large_Burst_Close_03}}
 {id=WavePlayer_2, properties {soundWaveAssetPtr=/Game/Sounds/Fort_Swamp_Sounds/MudBlobA001}}
 {id=WavePlayer_20, properties {soundWaveAssetPtr=/Game/Sounds/Fort_Weapons/Impacts/Mud/Mud_Explo_Large_Burst_Close_04}}
 {id=WavePlayer_8, properties {soundWaveAssetPtr=/Game/Sounds/Fort_Swamp_Sounds/MudBlobA002}}
 {id=WavePlayer_9, properties {soundWaveAssetPtr=/Game/Sounds/Fort_Swamp_Sounds/MudBlobA003}}
]
properties {
 attenuationSettings=&Fort_Traps_Att
 bOverrideConcurrency=true
 bPrimeOnLoad=true
 maxDistance=6000
 soundClassObject=&Fort_Traps
 volumeMultiplier=1.5
 concurrencyOverrides {bLimitToOwner=true, maxCount=1, resolutionRule=PreventNew}
}
