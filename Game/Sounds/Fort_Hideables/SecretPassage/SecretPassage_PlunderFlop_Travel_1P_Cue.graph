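# First-person Sound Cue for traveling through the PlunderFlop secret passage hideable.
# A mixer layers a sand-tunnel burrow travel loop, a heavily boosted Slurp vat ambience loop and a random one of four small mud burst sounds.
# Envelopers fade both loops in over about half a second, while the burst has no envelope.
# It primes on load and uses the Fort_Traps sound class and attenuation.
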SoundCue /Game/Sounds/Fort_Hideables/SecretPassage/SecretPassage_PlunderFlop_Travel_1P_Cue
root=Mixer_0
nodes [
 {
  id=Enveloper_0
  children [WavePlayer_2]
  properties {pitchCurve {editorCurveData {keys [{time=0}]}}, volumeCurve {editorCurveData {keys [<at 0 0>,<at 0.065 0.0145>,<at 0.5408 0.9972>]}}}
 }
 {
  id=Enveloper_1
  children [WavePlayer_3]
  properties {pitchCurve {editorCurveData {keys [{time=0}]}}, volumeCurve {editorCurveData {keys [<at 0 0>,<at 0.065 -0.0148>,<at 0.5382 0.9525>]}}}
 }
 {id=Mixer_0, children [Enveloper_0,Enveloper_1,Random_2], properties {inputVolume [0.8,4,1]}}
 {id=Random_2, children [WavePlayer_13, WavePlayer_14, WavePlayer_15, WavePlayer_16], properties {weights [1,1,1,1]}}
 {id=WavePlayer_13, properties {soundWaveAssetPtr=/Game/Sounds/Fort_Weapons/Impacts/Mud/Mud_Explo_Small_Burst_Close_01}}
 {id=WavePlayer_14, properties {soundWaveAssetPtr=/Game/Sounds/Fort_Weapons/Impacts/Mud/Mud_Explo_Small_Burst_Close_02}}
 {id=WavePlayer_15, properties {soundWaveAssetPtr=/Game/Sounds/Fort_Weapons/Impacts/Mud/Mud_Explo_Small_Burst_Close_03}}
 {id=WavePlayer_16, properties {soundWaveAssetPtr=/Game/Sounds/Fort_Weapons/Impacts/Mud/Mud_Explo_Small_Burst_Close_04}}
 {id=WavePlayer_2, properties {bLooping=true, soundWaveAssetPtr=/Game/Athena/Sounds/SandTunnels/SandTunnel_Burrowed_Travel_Enemy_Loop_01}}
 {id=WavePlayer_3, properties {bLooping=true, soundWaveAssetPtr=/Game/Athena/Sounds/Items/Slurp/Item_Slurp_Vat_Amb_Loop_Open_01}}
]
properties {
 attenuationSettings=&Fort_Traps_Att
 bPrimeOnLoad=true
 maxDistance=6000
 soundClassObject=&Fort_Traps
 volumeMultiplier=2
}
