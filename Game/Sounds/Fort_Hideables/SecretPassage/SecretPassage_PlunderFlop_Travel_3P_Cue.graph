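# Sound cue for the third-person travel sound of the PlunderFlop secret passage hideable.
# A mixer layers one of three equally weighted, randomly chosen passby waves, pitched up, over a quieter, pitched-down swamp mud blob.
# Each layer's modulator uses a fixed pitch rather than a random range.
# The cue primes on load and uses the trap attenuation settings and sound class, with boosted volume.
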SoundCue /Game/Sounds/Fort_Hideables/SecretPassage/SecretPassage_PlunderFlop_Travel_3P_Cue
root=Mixer_0
nodes [
 {id=Mixer_0, children [Modulator_0,Modulator_1], properties {inputVolume [0.7,0.2]}}
 {id=Modulator_0, children [Random_1], properties {pitchMax=2, pitchMin=2}}
 {id=Modulator_1, children [WavePlayer_5], properties {pitchMax=0.7, pitchMin=0.7}}
 {id=Random_1, children [WavePlayer_1,WavePlayer_2,WavePlayer_3], properties {weights [1,1,1]}}
 {id=WavePlayer_1, properties {soundWaveAssetPtr=&/3P_SecretPassage_Passyby_01}}
 {id=WavePlayer_2, properties {soundWaveAssetPtr=&/3P_SecretPassage_Passyby_02}}
 {id=WavePlayer_3, properties {soundWaveAssetPtr=&/3P_SecretPassage_Passyby_03}}
 {id=WavePlayer_5, properties {soundWaveAssetPtr=/Game/Sounds/Fort_Swamp_Sounds/MudBlobA003}}
]
properties {
 attenuationSettings=&Fort_Traps_Att
 bPrimeOnLoad=true
 maxDistance=6000
 soundClassObject=&Fort_Traps
 volumeMultiplier=2.4
}
